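# Animation Blueprint for the SpaceFeline backpack, animating idle and reactions in match and frontend.
# A bool blends between Backpack States, where anim input events trigger damage and elimination reactions, and a frontend idle/reactive machine.
# The backpack part modifier enables frontend reactivity, which repeats on delays via a state-entered notify.
# The limb AnimDynamics chain, testing machine, and face-material/health-polling logic are not wired to the output or any event.
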AnimBlueprint /BRCosmetics/Accessories/FORT_Backpacks/Backpack_F_MED_SpaceFeline/Meshes/F_MED_SpaceFeline_Pack_AnimBP
target_skeleton=&/F_MED_SpaceFeline_Pack_Skeleton
graphs [
 AnimGraph {nodes [
   AnimGraphNode_StateMachine_1 = Backpack States() {is=title AnimGraphNode_StateMachine}
   AnimGraphNode_StateMachine_2 = FE Idle + Reactive() {is=title AnimGraphNode_StateMachine}
   Output Pose(Result=Blend Poses by bool(
     BlendPose_0=@AnimGraphNode_StateMachine_2
     BlendPose_1=@AnimGraphNode_StateMachine_1
     bActiveValue=Get bShouldDisplayFrontendReactivity() -> (bShouldDisplayFrontendReactivity:bool)
    ) {blendType=Linear}) {layerGroup=None}
   Save cached pose 'Start_Pose'(Pose=Input Pose()) {cacheName=Start_Pose}
   n1 = Anim Dynamics - Bone: leg_l(ComponentPose=Local To Component(LocalPose=Use cached pose 'Start_Pose'()), Alpha=1, ExternalForce="0,0,0") {
    angularDampingOverride=1
    angularSpringConstant=20000
    bAngularSpring=true
    bOverrideAngularDamping=true
    bUseGravityOverride=true
    gravityScale=0.35
    simulationSpace=RootRelative
    alphaBoolBlend {blendOption=Linear}
    boundBone {boneName=leg_l}
    componentLinearAccScale=<1 1 1>
    componentLinearVelScale=<1 1 1>
    gravityOverride {z=-980}
    physicsBodyDefinitions [{
      boundBone {boneName=leg_l}
      boxExtents=<20 30 30>
      constraintSetup {angularLimitsMax=<15 15 15>, angularLimitsMin=<-15 -15 -15>, angularTarget=<1 0 0>}
      localJointOffset=<4 0 0>
     }]
   }
   n2 = Anim Dynamics - Bone: leg_r(ComponentPose=@n1, Alpha=1, ExternalForce="0,0,0") {
    angularDampingOverride=1
    angularSpringConstant=20000
    bAngularSpring=true
    bOverrideAngularDamping=true
    bUseGravityOverride=true
    gravityScale=0.35
    simulationSpace=RootRelative
    alphaBoolBlend {blendOption=Linear}
    boundBone {boneName=leg_r}
    componentLinearAccScale=<1 1 1>
    componentLinearVelScale=<1 1 1>
    gravityOverride {z=-980}
    physicsBodyDefinitions [{
      boundBone {boneName=leg_r}
      boxExtents=<20 30 30>
      constraintSetup {angularLimitsMax=<15 15 15>, angularLimitsMin=<-15 -15 -15>, angularTarget=<1 0 0>}
      localJointOffset=<4 0 0>
     }]
   }
   n3 = Anim Dynamics - Bone: arm_l(ComponentPose=@n2, Alpha=1, ExternalForce="0,0,0") {
    angularDampingOverride=1
    angularSpringConstant=25000
    bAngularSpring=true
    bOverrideAngularDamping=true
    bUseGravityOverride=true
    gravityScale=0.35
    simulationSpace=RootRelative
    alphaBoolBlend {blendOption=Linear}
    boundBone {boneName=arm_l}
    componentLinearAccScale=<1 1 1>
    componentLinearVelScale=<1 1 1>
    gravityOverride {z=-980}
    physicsBodyDefinitions [{
      boundBone {boneName=arm_l}
      boxExtents=<20 30 30>
      constraintSetup {angularLimitsMax=<15 30 15>, angularLimitsMin=<-15 -30 -15>, angularTarget=<1 0 0>}
      localJointOffset=<4 0 0>
     }]
   }
   n4 = Anim Dynamics - Bone: arm_r(ComponentPose=@n3, Alpha=1, ExternalForce="0,0,0") {
    angularDampingOverride=1
    angularSpringConstant=25000
    bAngularSpring=true
    bOverrideAngularDamping=true
    bUseGravityOverride=true
    gravityScale=0.35
    simulationSpace=RootRelative
    alphaBoolBlend {blendOption=Linear}
    boundBone {boneName=arm_r}
    componentLinearAccScale=<1 1 1>
    componentLinearVelScale=<1 1 1>
    gravityOverride {z=-980}
    physicsBodyDefinitions [{
      boundBone {boneName=arm_r}
      boxExtents=<20 30 30>
      constraintSetup {angularLimitsMax=<15 30 15>, angularLimitsMin=<-15 -30 -15>, angularTarget=<1 0 0>}
      localJointOffset=<4 0 0>
     }]
   }
   Save cached pose 'State_pose'(Pose=Component To Local(ComponentPose=@n4)) {cacheName=State_pose}
   AnimGraphNode_StateMachine_0 = Animation Testing() {is=title AnimGraphNode_StateMachine}
   Use cached pose 'State_pose'()
  ]}
 Animation Testing {
  entry=Idle
  path=AnimGraph.AnimGraphNode_StateMachine_0.Animation Testing
  ids=id
  nodes [
   AnimStateNode_0 = Idle()
   AnimStateNode_1 = Elimination() {is=title AnimStateNode}
   AnimStateNode_2 = HitReaction() {is=title AnimStateNode}
   AnimStateTransitionNode_0 = AnimStateTransitionNode() {bAutomaticRuleBasedOnSequencePlayerInState=true, crossfadeDuration=0}
   AnimStateTransitionNode_2 = AnimStateTransitionNode() {bAutomaticRuleBasedOnSequencePlayerInState=true, crossfadeDuration=0}
   AnimStateTransitionNode_3 = AnimStateTransitionNode() {bAutomaticRuleBasedOnSequencePlayerInState=true, crossfadeDuration=0}
  ]
  links [
   {
    dst=HitReaction
    rule=AnimGraph.AnimGraphNode_StateMachine_0.Animation Testing.AnimStateTransitionNode_2.Transition
    src=Elimination
    via=AnimStateTransitionNode_2
   }
   {
    dst=Idle
    rule=AnimGraph.AnimGraphNode_StateMachine_0.Animation Testing.AnimStateTransitionNode_3.Transition
    src=HitReaction
    via=AnimStateTransitionNode_3
   }
   {
    dst=Elimination
    rule=AnimGraph.AnimGraphNode_StateMachine_0.Animation Testing.AnimStateTransitionNode_0.Transition
    src=Idle
    via=AnimStateTransitionNode_0
   }
  ]
 }
 Idle {
  path=AnimGraph.AnimGraphNode_StateMachine_0.Animation Testing.AnimStateNode_0.Idle
  nodes [
   Sequence Player 'JollyTroll_Idle_Anim_lookabout_PACK'() {
    bLoopAnimation=false
    sequence=/BRCosmetics/Accessories/FORT_Backpacks/Backpack_M_MED_JollyTroll/Animations/JollyTroll_Idle_Anim_lookabout_PACK
   }
   Output Animation Pose(Result=Sequence Player 'SpaceFeline_Idle_PACK'() {bLoopAnimation=false, sequence=/BRCosmetics/Accessories/FORT_Backpacks/Backpack_F_MED_SpaceFeline/Animations/SpaceFeline_Idle_PACK})
  ]
 }
 Elimination {
  path=AnimGraph.AnimGraphNode_StateMachine_0.Animation Testing.AnimStateNode_1.Elimination
  nodes [
   Sequence Player 'JollyTroll_Sneeze_Anim_PACK'() {bLoopAnimation=false, sequence=/BRCosmetics/Accessories/FORT_Backpacks/Backpack_M_MED_JollyTroll/Animations/JollyTroll_Sneeze_Anim_PACK}
   n1 = Sequence Player 'SpaceFeline_Elimination_PACK'() {bLoopAnimation=false, sequence=/BRCosmetics/Accessories/FORT_Backpacks/Backpack_F_MED_SpaceFeline/Animations/SpaceFeline_Elimination_PACK}
   Output Animation Pose(Result=@n1)
  ]
 }
 HitReaction {
  path=AnimGraph.AnimGraphNode_StateMachine_0.Animation Testing.AnimStateNode_2.HitReaction
  nodes [
   n1 = Sequence Player 'SpaceFeline_HitReaction_PACK'() {bLoopAnimation=false, sequence=/BRCosmetics/Accessories/FORT_Backpacks/Backpack_F_MED_SpaceFeline/Animations/SpaceFeline_HitReaction_PACK}
   Output Animation Pose(Result=@n1)
  ]
 }
 Transition {path=AnimGraph.AnimGraphNode_StateMachine_0.Animation Testing.AnimStateTransitionNode_0.Transition, nodes [Result()]}
 Transition {path=AnimGraph.AnimGraphNode_StateMachine_0.Animation Testing.AnimStateTransitionNode_2.Transition, nodes [Result()]}
 Transition {path=AnimGraph.AnimGraphNode_StateMachine_0.Animation Testing.AnimStateTransitionNode_3.Transition, nodes [Result()]}
 Backpack States {
  entry=Idle
  path=AnimGraph.AnimGraphNode_StateMachine_1.Backpack States
  ids=id
  nodes [
   AnimStateNode_0 = Idle() {stateEntered {guid=#1}}
   AnimStateNode_1 = Took Damage() {is=title AnimStateNode}
   AnimStateNode_4 = Elimination()
   AnimStateTransitionNode_11 = AnimStateTransitionNode() {crossfadeDuration=0}
   AnimStateTransitionNode_12 = AnimStateTransitionNode() {crossfadeDuration=0}
   AnimStateTransitionNode_13 = AnimStateTransitionNode() {bAutomaticRuleBasedOnSequencePlayerInState=true, crossfadeDuration=0}
   AnimStateTransitionNode_5 = AnimStateTransitionNode() {bAutomaticRuleBasedOnSequencePlayerInState=true, crossfadeDuration=0}
  ]
  links [
   {
    dst=Idle
    rule=AnimGraph.AnimGraphNode_StateMachine_1.Backpack States.AnimStateTransitionNode_5.Transition
    src=Elimination
    via=AnimStateTransitionNode_5
   }
   {
    dst=Elimination
    rule=AnimGraph.AnimGraphNode_StateMachine_1.Backpack States.AnimStateTransitionNode_11.Transition
    src=Idle
    via=AnimStateTransitionNode_11
   }
   {
    dst=Took Damage
    rule=AnimGraph.AnimGraphNode_StateMachine_1.Backpack States.AnimStateTransitionNode_12.Transition
    src=Idle
    via=AnimStateTransitionNode_12
   }
   {
    dst=Idle
    rule=AnimGraph.AnimGraphNode_StateMachine_1.Backpack States.AnimStateTransitionNode_13.Transition
    src=Took Damage
    via=AnimStateTransitionNode_13
   }
  ]
 }
 Idle {
  path=AnimGraph.AnimGraphNode_StateMachine_1.Backpack States.AnimStateNode_0.Idle
  nodes [Output Animation Pose(Result=Sequence Player 'SpaceFeline_Idle_PACK'() {sequence=&SpaceFeline_Idle_PACK})]
 }
 Took Damage {
  path=AnimGraph.AnimGraphNode_StateMachine_1.Backpack States.AnimStateNode_1.Took Damage
  nodes [
   Sequence Player 'SharpMagnet_Backpack_Pain'() {bLoopAnimation=false, sequence=/BRCosmetics/Accessories/FORT_Backpacks/Backpack_M_MED_SharpMagnet/Animation/SharpMagnet_Backpack_Pain}
   n1 = Sequence Player 'SpaceFeline_HitReaction_PACK'() {bLoopAnimation=false, sequence=&SpaceFeline_HitReaction_PACK}
   Output Animation Pose(Result=@n1)
  ]
 }
 Elimination {
  path=AnimGraph.AnimGraphNode_StateMachine_1.Backpack States.AnimStateNode_4.Elimination
  nodes [
   n1 = Sequence Player 'SpaceFeline_Elimination_PACK'() {bLoopAnimation=false, sequence=&SpaceFeline_Elimination_PACK}
   Output Animation Pose(Result=@n1)
  ]
 }
 Transition {
  path=AnimGraph.AnimGraphNode_StateMachine_1.Backpack States.AnimStateTransitionNode_11.Transition
  nodes [Result(bCanEnterTransition=Get bEnemyPlayerEliminated() -> (bEnemyPlayerEliminated:bool))]
 }
 Transition {
  path=AnimGraph.AnimGraphNode_StateMachine_1.Backpack States.AnimStateTransitionNode_12.Transition
  nodes [Result(bCanEnterTransition=Get TookDamage() -> (TookDamage:bool))]
 }
 Transition {path=AnimGraph.AnimGraphNode_StateMachine_1.Backpack States.AnimStateTransitionNode_13.Transition, nodes [Result()]}
 Transition {path=AnimGraph.AnimGraphNode_StateMachine_1.Backpack States.AnimStateTransitionNode_5.Transition, nodes [Result()]}
 FE Idle + Reactive {
  entry=FE_Idle
  path=AnimGraph.AnimGraphNode_StateMachine_2.FE Idle + Reactive
  ids=id
  nodes [
   AnimStateNode_0 = FE_Idle() {is=title AnimStateNode}
   AnimStateNode_1 = FE_Reactive() {is=title AnimStateNode, stateEntered {guid=#2, notifyName=Entered FE_Reactive_Anim}}
   AnimStateTransitionNode_0 = AnimStateTransitionNode()
   AnimStateTransitionNode_1 = AnimStateTransitionNode() {bAutomaticRuleBasedOnSequencePlayerInState=true}
  ]
  links [
   {
    dst=FE_Reactive
    rule=AnimGraph.AnimGraphNode_StateMachine_2.FE Idle + Reactive.AnimStateTransitionNode_0.Transition
    src=FE_Idle
    via=AnimStateTransitionNode_0
   }
   {
    dst=FE_Idle
    rule=AnimGraph.AnimGraphNode_StateMachine_2.FE Idle + Reactive.AnimStateTransitionNode_1.Transition
    src=FE_Reactive
    via=AnimStateTransitionNode_1
   }
  ]
 }
 FE_Idle {
  path=AnimGraph.AnimGraphNode_StateMachine_2.FE Idle + Reactive.AnimStateNode_0.FE_Idle
  nodes [Output Animation Pose(Result=Sequence Player 'SpaceFeline_Idle_PACK'() {sequence=&SpaceFeline_Idle_PACK})]
 }
 FE_Reactive {
  path=AnimGraph.AnimGraphNode_StateMachine_2.FE Idle + Reactive.AnimStateNode_1.FE_Reactive
  nodes [
   n1 = Sequence Player 'SpaceFeline_Elimination_PACK'() {bLoopAnimation=false, sequence=&SpaceFeline_Elimination_PACK}
   Output Animation Pose(Result=@n1)
  ]
 }
 Transition {
  path=AnimGraph.AnimGraphNode_StateMachine_2.FE Idle + Reactive.AnimStateTransitionNode_0.Transition
  nodes [Result(bCanEnterTransition=Get Play Reactive FE Anim() -> (Play Reactive FE Anim:bool))]
 }
 Transition {path=AnimGraph.AnimGraphNode_StateMachine_2.FE Idle + Reactive.AnimStateTransitionNode_1.Transition, nodes [Result()]}
 EventGraph {nodes [
   n1 = Get Dynamic Material Instance(SkeletalMeshComponent=GetPartSkeletalMeshComponent(), ElementIndex=1)
   > SetScalarParameterValue(self=@n1, ParameterName=Default Open Mouth Netural Frame, Value=3)
   > SetScalarParameterValue(self=@n1, ParameterName=Default Open Eye Netural Frame, Value=3)
   > SetScalarParameterValue(self=@n1, ParameterName=Sad Mouth Frame, Value=0)
   > SetScalarParameterValue(self=@n1, ParameterName=Sad Eye Frame, Value=0)
   > SetScalarParameterValue(self=@n1, ParameterName=Blink Mouth Frame, Value=2)
   > SetScalarParameterValue(self=@n1, ParameterName=Blink Eye Frame, Value=2)
   > SetScalarParameterValue(self=@n1, ParameterName=Happy Mouth Frame, Value=1)
   > SetScalarParameterValue(self=@n1, ParameterName=Happy Eye Frame, Value=1)
   n2 = Cast To FortPlayerPawn(Object=GetOwningActor())
   n3 = GetHealth!(self:Fort Pawn Object Reference=@n2) -> (ReturnValue:float)
   > Set Current Health(Current Health:double=@n3) -> (Output_Get:double)
   > n4 = Branch(Condition="Not Equal (Float)"(A=@n3, B=Get Previous Health() -> (Previous Health:double)))
   > Set TookDamage(TookDamage:bool=true) -> (Output_Get:bool)
   > Set Previous Health(Previous Health:double=Get Current Health() -> (Current Health:double)) -> (Output_Get:double)
   Set TookDamage(execute=@n4.else, TookDamage=false)
   Event OnCharacterCustomizationLoaded()
   > n5 = Cast To FortPlayerPawn(Object=TryGetFortPawn())
   > n6 = Cast To B_Athena_PartModifier_Backpack_SpaceFeline!(
    execute
    Object=GetModifierForPartType!(self:Fort Player Pawn Object Reference=@n5, PartType:EFortCustomPartType Enum=Backpack) -> (ReturnValue:Custom Character Part Modifier Object Reference)
   ) -> (then, CastFailed, AsB Athena Part Modifier Backpack Space Feline:B Athena Part Modifier Backpack Space Feline Object Reference) {is=title K2Node_DynamicCast}
   > n7 = Set Reactivity Enabled(Reactivity Enabled:bool=Get ReactivityEnabled!(self:B Athena Part Modifier Backpack Space Feline Object Reference=@n6) -> (ReactivityEnabled:bool)) -> (Output_Get:bool)
   > n8 = Branch(Condition=@n7)
   > Do Once(Reset=@n10.then)
   > n9 = Delay(Duration=4)
   Set Play Reactive FE Anim(execute=[@n9.then, @n11.then], Play Reactive FE Anim:bool=true) -> (Output_Get:bool)
   n10 = Set Play Reactive FE Anim(execute=@n8.else, Play Reactive FE Anim=false)
   AnimNotify_Entered FE_Reactive_Anim() {is=title K2Node_Event}
   > Set Play Reactive FE Anim(Play Reactive FE Anim=false)
   > n11 = Delay(Duration=12)
   Event BlueprintPostEvaluateAnimation() {is=title K2Node_Event}
   > Set bEnemyPlayerEliminated(bEnemyPlayerEliminated:bool=false) -> (Output_Get:bool)
   n12 = Event BlueprintProcessAnimInputEvent()
   > n13 = Sequence()
   > Branch(Condition=IsNamed(self=@n12.AnimInputEvent, InEventName=aie.EnemyPlayerEliminated))
   > Set bEnemyPlayerEliminated(bEnemyPlayerEliminated=true)
   Branch(
    execute=@n13.then_1
    Condition=OR Boolean(
     A=IsNamed(self=@n12.AnimInputEvent, InEventName=aie.PlayerVitality.HealthDecreased)
     B=IsNamed(self=@n12.AnimInputEvent, InEventName=Noneaie.PlayerVitality.ShieldsDecreased)
    )
   )
   > Set TookDamage(TookDamage=true)
   > RetriggerableDelay(Duration=0.2)
   > Set TookDamage(TookDamage=false)
   TryGetPawnOwner()
   Event BlueprintUpdateAnimation()
  ]}
]
variables [
 Active_Kill {type=boolean}
 As B Athena Part Modifier Backpack Space Feline {type=B_Athena_PartModifier_Backpack_SpaceFeline_C, default=null}
 bEnemyPlayerEliminated {type=boolean}
 Current Health {type=number}
 "FInterp(Alpha, (10: 10))" {type=number}
 "FInterp(Alpha, (10: 10))_0" {type=number}
 float_arm_l {type=Vector, default=<-0.4969 0 -0.4969>}
 float_arm_r {type=Vector, default=<-0.4969 0 -0.4969>}
 float_body {type=Vector, default=<0 0 -0.4809>}
 float_leg_l {type=Vector, default=<-0.3313 0 -0.3313>}
 float_leg_r {type=Vector, default=<-0.3313 0 -0.3313>}
 Happy_State {type=number}
 KillCount {type=number}
 Play Reactive FE Anim {type=boolean}
 Previous Health {type=number}
 Reactivity Enabled {type=boolean}
 Sad_State {type=number}
 Skel_Mesh {type=SkeletalMeshComponent}
 TookDamage {type=boolean}
 Wave {type=Rotator, default=<rot -19.661 0 0>}
 Wave_0 {type=Vector}
]
